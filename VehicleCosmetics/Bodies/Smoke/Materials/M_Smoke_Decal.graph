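# Masked, two-sided Default Lit material for decals on the Smoke vehicle body cosmetic.
# Its only listed node is a texture sample of T_Smoke_Decals, with no connections to material outputs shown.
# Usage is enabled for skeletal meshes and mesh deformers, and automatic usage setting in the editor is off.
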Material /VehicleCosmetics/Bodies/Smoke/Materials/M_Smoke_Decal
graph {
 properties {
  bAutomaticallySetUsageInEditor=false
  bUsedWithMeshDeformer=true
  bUsedWithSkeletalMesh=true
  blendMode=BLEND_Masked
  twoSided=true
  usedShadingModels=Default Lit
 }
 nodes [TextureSample() {texture=/VehicleCosmetics/Bodies/Smoke/Textures/T_Smoke_Decals}]
}
